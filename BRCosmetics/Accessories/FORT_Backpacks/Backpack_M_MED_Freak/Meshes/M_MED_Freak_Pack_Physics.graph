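# Physics asset for the M_MED_Freak backpack cosmetic, giving physics bodies to three frog figures (left, middle, right) attached to a pack.
# Each frog has a capsule body and limbs with a sphere head, and the middle frog adds a three-segment tongue; the pack is one thin box.
# Limbs, heads and tongue chain to their parents with angle-limited constraints, while each frog body joins the pack with free swing and twist.
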PhysicsAsset /BRCosmetics/Accessories/FORT_Backpacks/Backpack_M_MED_Freak/Meshes/M_MED_Freak_Pack_Physics
bodies {
 frogLeft_ankle_l [{
   shapeType=Capsule
   length=2.2236
   radius=0.9427
   shapeName=frogLeft_ankle_l_capsule
   center=<1.1613 0.2366 0.0493>
   rotation=<rot 0 92.945 -84.7011>
  }]
 frogLeft_ankle_r [{
   shapeType=Capsule
   length=1.9931
   radius=0.9838
   shapeName=frogLeft_ankle_r_capsule
   center=<1.5338 0.3287 -0.0767>
   rotation=<rot 0 -90.7009 -87.6778>
  }]
 frogLeft_arm_l [{shapeType=Capsule, length=1.1494, radius=1.3649, center=<1.6018 -0.303 -0.2511>, rotation=<rot 90 0 0>}]
 frogLeft_arm_r [{shapeType=Capsule, length=0.7764, radius=1.3113, center=<1.7838 -0.1943 -0.1248>, rotation=<rot -90 1.8985 -1.8984>}]
 frogLeft_ball_l [{shapeType=Capsule, length=1.505, radius=1.255, shapeName=frogLeft_ball_l_capsule, center=<1.6965 0.061 0>, rotation=<rot 0 -90 89.9994>}]
 frogLeft_ball_r [{
   shapeType=Capsule
   length=1.505
   radius=1.255
   shapeName=frogLeft_ball_r_capsule
   center=<1.6333 0.1268 0>
   rotation=<rot 0 -89.9999 89.9997>
  }]
 frogLeft_body [{shapeType=Capsule, length=2.4136, radius=4.9775, center=<4.3336 1.737 0.2395>, rotation=<rot 0 0 -90>}]
 frogLeft_elbow_l [{shapeType=Capsule, length=1.8967, radius=1.0766, center=<1.7908 -0.0185 -0.1917>, rotation=<rot 84.9997 -179.9999 -179.9999>}]
 frogLeft_elbow_r [{shapeType=Capsule, length=2.3972, radius=1.0746, center=<2.2791 -0.3636 -0.0246>, rotation=<rot 90 0 0>}]
 frogLeft_head [{shapeType=Sphere, radius=5.4179, shapeName=frogLeft_head_sphere, center=<2.4766 1.5976 -0.1427>}]
 frogLeft_knee_l [{shapeType=Capsule, length=1.7319, radius=1.1464, center=<1.8361 -0.1264 -0.1102>, rotation=<rot 90 0 0>}]
 frogLeft_knee_r [{shapeType=Capsule, length=1.373, radius=1.3594, center=<1.6316 0.0204 0.0304>, rotation=<rot 90 0 0>}]
 frogLeft_leg_l [{shapeType=Capsule, length=2.7533, radius=1.9003, center=<1.3758 -0.3999 -0.3313>, rotation=<rot 84.9988 -179.9998 -174.9999>}]
 frogLeft_leg_r [{shapeType=Capsule, length=2.5094, radius=1.7148, center=<2.3693 -0.7737 0.1357>, rotation=<rot 90 0 0>}]
 frogLeft_wrist_l [{shapeType=Capsule, length=0, radius=2.25, center=<2.4863 0.9393 -0.2997>}]
 frogLeft_wrist_r [{shapeType=Capsule, length=0, radius=2.25, center=<2.4506 1.07 0.8065>}]
 frogMiddle_ankle_l [{
   shapeType=Capsule
   length=2.0397
   radius=1.0708
   shapeName=frogMiddle_ankle_l_capsule
   center=<0.9486 0.0413 0.2279>
   rotation=<rot 0 95.7948 -84.1354>
  }]
 frogMiddle_ankle_r [{
   shapeType=Capsule
   length=1.3593
   radius=1.1038
   shapeName=frogMiddle_ankle_r_capsule
   center=<1.0608 0.1767 -0.1485>
   rotation=<rot 0 -86.3004 -87.2156>
  }]
 frogMiddle_arm_l [{shapeType=Capsule, length=0.4845, radius=1.4395, center=<1.635 -0.2809 -0.3099>, rotation=<rot 90 0 0>}]
 frogMiddle_arm_r [{shapeType=Capsule, length=1.3976, radius=1.2554, center=<1.6821 0.2332 0.2937>, rotation=<rot 90 0 0>}]
 frogMiddle_ball_l [{
   shapeType=Capsule
   length=2.5
   radius=1.5
   shapeName=frogMiddle_ball_l_capsule
   center=<1.7973 0.6213 0.6064>
   rotation=<rot 0.0001 -89.9999 80.0002>
  }]
 frogMiddle_ball_r [{shapeType=Capsule, length=2.5, radius=1.5, shapeName=frogMiddle_ball_r_capsule, center=<1.9426 0.6213 0>, rotation=<rot 0 -89.9999 90.0004>}]
 frogMiddle_body [{shapeType=Capsule, length=3.5435, radius=4.6802, center=<3.2053 2.6416 -0.1332>, rotation=<rot 0 0 -100.0001>}]
 frogMiddle_elbow_l [{shapeType=Capsule, length=2.2087, radius=1.1123, center=<2.3831 -0.3794 -0.1664>, rotation=<rot 84.9998 -179.9999 -179.9999>}]
 frogMiddle_elbow_r [{shapeType=Capsule, length=2.235, radius=1.2001, center=<2.0847 -0.4053 0.1138>, rotation=<rot 90 0 0>}]
 frogMiddle_head [{shapeType=Sphere, radius=6.1074, shapeName=frogMiddle_head_sphere, center=<2.0732 2.1768 -0.4322>}]
 frogMiddle_knee_l [{shapeType=Capsule, length=1.6346, radius=1.1544, center=<1.9746 -0.2385 0.0258>, rotation=<rot 74.9993 89.9998 89.9998>}]
 frogMiddle_knee_r [{shapeType=Capsule, length=1.7785, radius=1.3065, center=<1.9298 -0.0052 0.0488>, rotation=<rot 90 0 -5.0001>}]
 frogMiddle_leg_l [{shapeType=Capsule, length=2.7804, radius=1.7258, center=<2.3775 -0.216 -0.3786>, rotation=<rot 90 0 0>}]
 frogMiddle_leg_r [{shapeType=Capsule, length=2.75, radius=1.4825, center=<2.9094 -0.8519 -0.1529>, rotation=<rot -0.0003 89.9987 -89.9998>}]
 frogMiddle_tongue1 [{
   shapeType=Capsule
   length=2.6559
   radius=0.75
   shapeName=frogMiddle_tongue1_capsule
   center=<2.0478 0.0878 0.0906>
   rotation=<rot 0 -92.1419 -88.1064>
  }]
 frogMiddle_tongue2 [{
   shapeType=Capsule
   length=3.013
   radius=0.75
   shapeName=frogMiddle_tongue2_capsule
   center=<2.0583 0.0317 0.0042>
   rotation=<rot 0 -91.6629 -89.144>
  }]
 frogMiddle_tongue3 [{
   shapeType=Capsule
   length=0.755
   radius=0.505
   shapeName=frogMiddle_tongue3_capsule
   center=<0.9953 0 0.0586>
   rotation=<rot 0 -89.9998 90.0008>
  }]
 frogMiddle_wrist_l [{shapeType=Capsule, length=0, radius=2.25, center=<2.3527 0.7126 -0.4454>}]
 frogMiddle_wrist_r [{shapeType=Capsule, length=0, radius=2.25, center=<2.0796 0.8595 0.623>}]
 frogRight_ankle_l [{
   shapeType=Capsule
   length=1.9644
   radius=1.0177
   shapeName=frogRight_ankle_l_capsule
   center=<1.0224 0.6218 -0.0587>
   rotation=<rot 1.4477 77.167 -86.0183>
  }]
 frogRight_ankle_r [{
   shapeType=Capsule
   length=1.8071
   radius=0.767
   shapeName=frogRight_ankle_r_capsule
   center=<0.8635 0.6486 0.0572>
   rotation=<rot 0.4404 -101.7428 -87.5013>
  }]
 frogRight_arm_l [{shapeType=Capsule, length=1.1691, radius=1.33, center=<1.4344 -0.1039 -0.044>, rotation=<rot 90 0 0>}]
 frogRight_arm_r [{shapeType=Capsule, length=1.6309, radius=1.099, center=<1.7607 0.0695 0.2131>, rotation=<rot 90 0 0>}]
 frogRight_ball_l [{
   shapeType=Capsule
   length=3
   radius=1
   shapeName=frogRight_ball_l_capsule
   center=<3.6157 0.5026 0.3323>
   rotation=<rot 0.0001 -79.9996 89.9993>
  }]
 frogRight_ball_r [{shapeType=Capsule, length=3, radius=1.25, shapeName=frogRight_ball_r_capsule, center=<2.2655 0 -0.0001>, rotation=<rot 0 -89.9999 90.0004>}]
 frogRight_body [{shapeType=Capsule, length=2.5351, radius=4.7205, center=<4.3261 -2.7028 0.2251>, rotation=<rot 79.999 10.0017 -0.0002>}]
 frogRight_elbow_l [{shapeType=Capsule, length=2.1473, radius=1.2099, center=<2.0286 -0.2474 -0.2023>, rotation=<rot 85 174.9999 -179.9999>}]
 frogRight_elbow_r [{shapeType=Capsule, length=2.5388, radius=1.0802, center=<1.9843 0.0162 0.063>, rotation=<rot 90 0 0>}]
 frogRight_head [{shapeType=Sphere, radius=5.6628, shapeName=frogRight_head_sphere, center=<2.7565 -1.1283 0.0794>}]
 frogRight_knee_l [{shapeType=Capsule, length=1.3745, radius=1.3956, center=<1.8761 -0.1058 -0.1109>, rotation=<rot 90 0 0>}]
 frogRight_knee_r [{shapeType=Capsule, length=1.6464, radius=1.1809, center=<1.8717 -0.1311 0.1398>, rotation=<rot 90 -90 -95>}]
 frogRight_leg_l [{shapeType=Capsule, length=1.5164, radius=1.7309, center=<1.9035 -0.7688 0.0071>, rotation=<rot 0 -94.9996 -90>}]
 frogRight_leg_r [{shapeType=Capsule, length=1.5747, radius=1.6551, center=<2.5283 -0.5685 0.2747>, rotation=<rot 72.0359 -34.2662 -32.9466>}]
 frogRight_wrist_l [{shapeType=Capsule, length=0, radius=2.25, center=<2.0372 0.981 -0.4976>}]
 frogRight_wrist_r [{shapeType=Capsule, length=0, radius=2, center=<2.2669 0.728 0.3166>, rotation=<rot 0.2185 29.3488 -0.2076>}]
 pack [{shapeType=Box, extentX=70.5402, extentY=1.4211, extentZ=42.5222, center=<-0.3706 1.5181 -6.9678>}]
}
constraints [
 {bone1Name=frogRight_arm_l, bone2Name=frogRight_body, swing1LimitDegrees=90, swing2LimitDegrees=30, twistLimitDegrees=5}
 {bone1Name=frogRight_elbow_l, bone2Name=frogRight_arm_l, swing1LimitDegrees=30, swing2LimitDegrees=30, twistLimitDegrees=5}
 {bone1Name=frogRight_wrist_l, bone2Name=frogRight_elbow_l, swing1LimitDegrees=15, swing2LimitDegrees=15, twistLimitDegrees=5}
 {bone1Name=frogRight_arm_r, bone2Name=frogRight_body, swing1LimitDegrees=75, swing2LimitDegrees=15, twistLimitDegrees=5}
 {bone1Name=frogRight_elbow_r, bone2Name=frogRight_arm_r, swing1LimitDegrees=30, swing2LimitDegrees=30, twistLimitDegrees=5}
 {bone1Name=frogRight_wrist_r, bone2Name=frogRight_elbow_r, swing1LimitDegrees=15, swing2LimitDegrees=15, twistLimitDegrees=5}
 {bone1Name=frogRight_leg_l, bone2Name=frogRight_body, swing1LimitDegrees=30, swing2LimitDegrees=30, twistLimitDegrees=0.5}
 {bone1Name=frogRight_knee_l, bone2Name=frogRight_leg_l, swing1LimitDegrees=25, swing2LimitDegrees=25, twistLimitDegrees=0.5}
 {bone1Name=frogRight_ankle_l, bone2Name=frogRight_knee_l, swing1LimitDegrees=13, swing2LimitDegrees=13, twistLimitDegrees=0.5}
 {bone1Name=frogRight_ball_l, bone2Name=frogRight_ankle_l, swing1LimitDegrees=13, swing2LimitDegrees=13, twistLimitDegrees=0.5}
 {bone1Name=frogRight_leg_r, bone2Name=frogRight_body, swing1LimitDegrees=40, swing2LimitDegrees=25, twistLimitDegrees=0.5}
 {bone1Name=frogRight_knee_r, bone2Name=frogRight_leg_r, swing1LimitDegrees=26, swing2LimitDegrees=26, twistLimitDegrees=0.5}
 {bone1Name=frogRight_ankle_r, bone2Name=frogRight_knee_r, swing1LimitDegrees=13, swing2LimitDegrees=13, twistLimitDegrees=0.5}
 {bone1Name=frogRight_ball_r, bone2Name=frogRight_ankle_r, swing1LimitDegrees=13, swing2LimitDegrees=13, twistLimitDegrees=5}
 {bone1Name=frogRight_head, bone2Name=frogRight_body, swing1LimitDegrees=35, swing2LimitDegrees=35, twistLimitDegrees=10}
 {bone1Name=frogMiddle_arm_l, bone2Name=frogMiddle_body, swing1LimitDegrees=100, swing2LimitDegrees=30, twistLimitDegrees=5}
 {bone1Name=frogMiddle_elbow_l, bone2Name=frogMiddle_arm_l, swing1LimitDegrees=30, swing2LimitDegrees=30, twistLimitDegrees=5}
 {bone1Name=frogMiddle_wrist_l, bone2Name=frogMiddle_elbow_l, swing1LimitDegrees=15, swing2LimitDegrees=15, twistLimitDegrees=5}
 {bone1Name=frogMiddle_arm_r, bone2Name=frogMiddle_body, swing1LimitDegrees=70, swing2LimitDegrees=30, twistLimitDegrees=5}
 {bone1Name=frogMiddle_elbow_r, bone2Name=frogMiddle_arm_r, swing1LimitDegrees=30, swing2LimitDegrees=30, twistLimitDegrees=5}
 {bone1Name=frogMiddle_wrist_r, bone2Name=frogMiddle_elbow_r, swing1LimitDegrees=15, swing2LimitDegrees=15, twistLimitDegrees=5}
 {bone1Name=frogMiddle_leg_l, bone2Name=frogMiddle_body, swing1LimitDegrees=30, swing2LimitDegrees=30, twistLimitDegrees=5}
 {bone1Name=frogMiddle_knee_l, bone2Name=frogMiddle_leg_l, swing1LimitDegrees=30, swing2LimitDegrees=30, twistLimitDegrees=5}
 {bone1Name=frogMiddle_ankle_l, bone2Name=frogMiddle_knee_l, swing1LimitDegrees=15, swing2LimitDegrees=15, twistLimitDegrees=5}
 {bone1Name=frogMiddle_ball_l, bone2Name=frogMiddle_ankle_l, swing1LimitDegrees=15, swing2LimitDegrees=15, twistLimitDegrees=5}
 {bone1Name=frogMiddle_leg_r, bone2Name=frogMiddle_body, swing1LimitDegrees=30, swing2LimitDegrees=30, twistLimitDegrees=5}
 {bone1Name=frogMiddle_knee_r, bone2Name=frogMiddle_leg_r, swing1LimitDegrees=30, swing2LimitDegrees=30, twistLimitDegrees=5}
 {bone1Name=frogMiddle_ankle_r, bone2Name=frogMiddle_knee_r, swing1LimitDegrees=15, swing2LimitDegrees=15, twistLimitDegrees=5}
 {bone1Name=frogMiddle_ball_r, bone2Name=frogMiddle_ankle_r, swing1LimitDegrees=15, swing2LimitDegrees=15, twistLimitDegrees=5}
 {bone1Name=frogMiddle_head, bone2Name=frogMiddle_body, swing1LimitDegrees=45, swing2LimitDegrees=45, twistLimitDegrees=10}
 {bone1Name=frogMiddle_tongue1, bone2Name=frogMiddle_head, swing1LimitDegrees=30, swing2LimitDegrees=10, twistLimitDegrees=1}
 {bone1Name=frogMiddle_tongue2, bone2Name=frogMiddle_tongue1, swing1LimitDegrees=20, swing2LimitDegrees=20, twistLimitDegrees=10}
 {bone1Name=frogMiddle_tongue3, bone2Name=frogMiddle_tongue2, swing1LimitDegrees=20, swing2LimitDegrees=20, twistLimitDegrees=5}
 {bone1Name=frogLeft_arm_l, bone2Name=frogLeft_body, swing1LimitDegrees=90, swing2LimitDegrees=45, twistLimitDegrees=10}
 {bone1Name=frogLeft_elbow_l, bone2Name=frogLeft_arm_l, swing1LimitDegrees=30, swing2LimitDegrees=30, twistLimitDegrees=5}
 {bone1Name=frogLeft_wrist_l, bone2Name=frogLeft_elbow_l, swing1LimitDegrees=15, swing2LimitDegrees=15, twistLimitDegrees=5}
 {bone1Name=frogLeft_arm_r, bone2Name=frogLeft_body, swing1LimitDegrees=75, swing2LimitDegrees=30, twistLimitDegrees=5}
 {bone1Name=frogLeft_elbow_r, bone2Name=frogLeft_arm_r, swing1LimitDegrees=30, swing2LimitDegrees=30, twistLimitDegrees=5}
 {bone1Name=frogLeft_wrist_r, bone2Name=frogLeft_elbow_r, swing1LimitDegrees=15, swing2LimitDegrees=15, twistLimitDegrees=5}
 {bone1Name=frogLeft_leg_r, bone2Name=frogLeft_body, swing1LimitDegrees=25, swing2LimitDegrees=25, twistLimitDegrees=5}
 {bone1Name=frogLeft_knee_r, bone2Name=frogLeft_leg_r, swing1LimitDegrees=30, swing2LimitDegrees=30, twistLimitDegrees=5}
 {bone1Name=frogLeft_ankle_r, bone2Name=frogLeft_knee_r, swing1LimitDegrees=17, swing2LimitDegrees=17, twistLimitDegrees=5}
 {bone1Name=frogLeft_ball_r, bone2Name=frogLeft_ankle_r, swing1LimitDegrees=17, swing2LimitDegrees=17, twistLimitDegrees=5}
 {bone1Name=frogLeft_leg_l, bone2Name=frogLeft_body, swing1LimitDegrees=30, swing2LimitDegrees=30, twistLimitDegrees=5}
 {bone1Name=frogLeft_knee_l, bone2Name=frogLeft_leg_l, swing1LimitDegrees=25, swing2LimitDegrees=25, twistLimitDegrees=5}
 {bone1Name=frogLeft_ankle_l, bone2Name=frogLeft_knee_l, swing1LimitDegrees=17, swing2LimitDegrees=17, twistLimitDegrees=5}
 {bone1Name=frogLeft_ball_l, bone2Name=frogLeft_ankle_l, swing1LimitDegrees=17, swing2LimitDegrees=17, twistLimitDegrees=5}
 {bone1Name=frogLeft_head, bone2Name=frogLeft_body, swing1LimitDegrees=55, swing2LimitDegrees=55, twistLimitDegrees=10}
 {
  bone1Name=frogRight_body
  bone2Name=pack
  swing1LimitDegrees=45
  swing1Motion=Free
  swing2LimitDegrees=45
  swing2Motion=Free
  twistLimitDegrees=45
  twistMotion=Free
 }
 {
  bone1Name=frogMiddle_body
  bone2Name=pack
  swing1LimitDegrees=45
  swing1Motion=Free
  swing2LimitDegrees=45
  swing2Motion=Free
  twistLimitDegrees=45
  twistMotion=Free
 }
 {
  bone1Name=frogLeft_body
  bone2Name=pack
  swing1LimitDegrees=45
  swing1Motion=Free
  swing2LimitDegrees=45
  swing2Motion=Free
  twistLimitDegrees=45
  twistMotion=Free
 }
]
