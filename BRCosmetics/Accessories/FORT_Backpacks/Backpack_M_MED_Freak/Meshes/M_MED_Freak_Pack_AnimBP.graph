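# Animation blueprint for the M_MED_Freak backpack that adds physics motion to the pack and poses its frog tongue.
# The input pose runs through a RigidBody node simulating in the pack bone's space with overridden gravity.
# Each update flags rigid body as disabled when it is switched off or the LOD is above 1.
# That flag enables an additive roll on the middle frog tongue bone, apparently a static fallback for the missing simulation.
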AnimBlueprint /BRCosmetics/Accessories/FORT_Backpacks/Backpack_M_MED_Freak/Meshes/M_MED_Freak_Pack_AnimBP
target_skeleton=&/M_MED_Freak_Pack_Skeleton
graphs [
 AnimGraph {nodes [
   AnimGraphNode_SubInput_1 = Input Pose()
   n1 = RigidBody(ComponentPose=Local To Component(LocalPose=@AnimGraphNode_SubInput_1)) {
    bClampLinearTranslationLimitToRefPose=true
    bOverrideWorldGravity=true
    overridePhysicsAsset=&/M_MED_Freak_Pack_Physics
    simulationSpace=BaseBoneSpace
    alphaBoolBlend {blendOption=Linear}
    baseBoneRef {boneName=pack}
    componentLinearAccScale=<0.85 0.85 0.85>
    overrideWorldGravity {z=-980}
   }
   n2 = "Transform (Modify) Bone - Bone: frogMiddle_tongue1"!(
    ComponentPose=@n1
    bAlphaBoolEnabled:bool=Get isRigidBodyDisabled() -> (isRigidBodyDisabled:bool)
    Translation="0,0,0"
    Rotation="0,0,-90"
    Scale="1,1,1"
   ) -> (Pose) {
    alphaInputType=Bool
    rotationMode=BMM_Additive
    alphaBoolBlend {blendOption=Linear}
    boneToModify {boneName=frogMiddle_tongue1}
    rotation {roll=-90}
   }
   Output Pose(Result=Component To Local(ComponentPose=@n2)) {layerGroup=None}
  ]}
 EventGraph {nodes [
   Event BlueprintUpdateAnimation()
   n1 = Get CurrentLOD()
   n2 = Get bRigidBodyEnabled()
   > Set isRigidBodyDisabled(isRigidBodyDisabled:bool=OR Boolean(A=NOT Boolean(A=@n2), B="integer > integer"(A=@n1, B=1) {is=title K2Node_CallFunction})) -> (Output_Get:bool)
   TryGetPawnOwner()
  ]}
]
variables [isRigidBodyDisabled {type=boolean}]
